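# Sound cue for entering a dumpster hideable, spatialized with the traps mono attenuation asset.
# A Random node picks between two equally weighted enter waves, then feeds a Modulator that applies a fixed volume boost, offset by a reduced cue-level volume multiplier.
# The root SoundClass node overrides the class to Fort_Traps, although the cue itself is assigned Fort_Gadgets_Items.
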SoundCue /FNGameplayCues/Sounds/Fort_Hideables/Hideable_Dumpster_Enter_Cue
root=SoundClass_0
nodes [
 {id=Modulator_0, children [Random_0], properties {volumeMax=2, volumeMin=2}}
 {id=Random_0, children [WavePlayer_1,WavePlayer_2], properties {bShouldExcludeFromBranchCulling=true, weights [1,1]}}
 {id=SoundClass_0, children [Modulator_0], properties {soundClassOverride=&Fort_Traps}}
 {id=WavePlayer_1, properties {soundWaveAssetPtr=&/Hideable_Dumpster_Enter_01}}
 {id=WavePlayer_2, properties {soundWaveAssetPtr=&/Hideable_Dumpster_Enter_02}}
]
properties {
 attenuationSettings=/Game/Sounds/Attenuation/Fort_Traps_Mono_Att
 maxDistance=6000
 soundClassObject=&Fort_Gadgets_Items
 volumeMultiplier=0.4286
}
