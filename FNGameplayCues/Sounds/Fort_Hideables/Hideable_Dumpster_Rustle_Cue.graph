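# Sound cue that plays a rustle sound for the dumpster hideable prop.
# A Random node picks one of three equally weighted rustle waves and is excluded from branch culling.
# The pick passes through a SoundClass node that overrides the cue's Fort_Gadgets_Items class with Fort_Traps, then a default Modulator.
# Playback uses the hiding-props rustle attenuation at reduced volume.
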SoundCue /FNGameplayCues/Sounds/Fort_Hideables/Hideable_Dumpster_Rustle_Cue
root=Modulator_0
nodes [
 {id=Modulator_0, children [SoundClass_0]}
 {id=Random_0, children [WavePlayer_1,WavePlayer_0,WavePlayer_2], properties {bShouldExcludeFromBranchCulling=true, weights [1,1,1]}}
 {id=SoundClass_0, children [Random_0], properties {soundClassOverride=&Fort_Traps}}
 {id=WavePlayer_0, properties {soundWaveAssetPtr=&/HIDEABLE_DUMPSTER_RUSTLE_02}}
 {id=WavePlayer_1, properties {soundWaveAssetPtr=&/HIDEABLE_DUMPSTER_RUSTLE_01}}
 {id=WavePlayer_2, properties {soundWaveAssetPtr=&/HIDEABLE_DUMPSTER_RUSTLE_03}}
]
properties {
 attenuationSettings=/FNGameplayCues/Sounds/Attenuation/Fort_HidingProps_Rustle
 maxDistance=3000
 soundClassObject=&Fort_Gadgets_Items
 volumeMultiplier=0.4286
}
