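# Sound cue for entering a haystack hideable, playing a single entry wave.
# A modulator adds slight random volume variation at a boosted level, offset by a reduced cue-wide volume multiplier.
# A sound class node overrides the cue's gadgets-and-items class with the traps class for the wave player beneath it.
# Spatialization uses the shared mono trap attenuation asset.
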SoundCue /FNGameplayCues/Sounds/Fort_Hideables/Hideable_HayStack_Enter_Cue
root=Modulator_0
nodes [
 {id=Modulator_0, children [SoundClass_0], properties {volumeMax=1.7, volumeMin=1.6}}
 {id=SoundClass_0, children [WavePlayer_1], properties {soundClassOverride=&Fort_Traps}}
 {id=WavePlayer_1, properties {soundWaveAssetPtr=&/Hideable_HayStack_Enter_v2_01}}
]
properties {
 attenuationSettings=/Game/Sounds/Attenuation/Fort_Traps_Mono_Att
 maxDistance=6000
 soundClassObject=&Fort_Gadgets_Items
 volumeMultiplier=0.4286
}
