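# Sound cue for exiting a haystack hideable, judging by its name, using the Fort traps mono attenuation.
# A Random node, excluded from branch culling, picks between two equally weighted exit waves.
# A SoundClass node overrides the cue's Gadgets_Items class with Fort_Traps for those waves.
# The root Modulator applies a fixed volume boost with no variation, largely offset by the cue's reduced volume multiplier.
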SoundCue /FNGameplayCues/Sounds/Fort_Hideables/Hideable_HayStack_Exit_Cue
root=Modulator_0
nodes [
 {id=Modulator_0, children [SoundClass_0], properties {volumeMax=2.5, volumeMin=2.5}}
 {id=Random_0, children [WavePlayer_1,WavePlayer_2], properties {bShouldExcludeFromBranchCulling=true, weights [1,1]}}
 {id=SoundClass_0, children [Random_0], properties {soundClassOverride=&Fort_Traps}}
 {id=WavePlayer_1, properties {soundWaveAssetPtr=&/Hideable_HayStack_Exit_v2_01}}
 {id=WavePlayer_2, properties {soundWaveAssetPtr=&/Hideable_HayStack_Exit_v2_02}}
]
properties {
 attenuationSettings=/Game/Sounds/Attenuation/Fort_Traps_Mono_Att
 maxDistance=6000
 soundClassObject=&Fort_Gadgets_Items
 volumeMultiplier=0.4286
}
