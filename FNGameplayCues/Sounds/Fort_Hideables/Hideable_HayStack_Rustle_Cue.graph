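# Sound cue for the rustle of a haystack hideable prop.
# A Random node picks between two equally weighted rustle waves, then feeds a SoundClass node that overrides the cue's Gadgets/Items class with Traps.
# A root Modulator adds slight volume variation, and playback uses the shared hiding-props rustle attenuation.
# A concurrency override allows one instance at a time and blocks new plays while it is active.
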SoundCue /FNGameplayCues/Sounds/Fort_Hideables/Hideable_HayStack_Rustle_Cue
root=Modulator_0
nodes [
 {id=Modulator_0, children [SoundClass_0], properties {volumeMax=1.5, volumeMin=1.3}}
 {id=Random_0, children [WavePlayer_0,WavePlayer_1], properties {bShouldExcludeFromBranchCulling=true, weights [1,1]}}
 {id=SoundClass_0, children [Random_0], properties {soundClassOverride=&Fort_Traps}}
 {id=WavePlayer_0, properties {soundWaveAssetPtr=&/Hideable_HayStack_Rustle_01}}
 {id=WavePlayer_1, properties {soundWaveAssetPtr=&/Hideable_HayStack_Rustle_02}}
]
properties {
 attenuationSettings=/FNGameplayCues/Sounds/Attenuation/Fort_HidingProps_Rustle
 bOverrideConcurrency=true
 maxDistance=3000
 soundClassObject=&Fort_Gadgets_Items
 volumeMultiplier=0.4286
 concurrencyOverrides {maxCount=1, resolutionRule=PreventNew}
}
